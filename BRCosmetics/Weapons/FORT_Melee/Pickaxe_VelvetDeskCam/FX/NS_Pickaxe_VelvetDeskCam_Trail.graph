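# Niagara system that draws the swing trail for the VelvetDeskCam pickaxe as short-lived ribbons.
# One emitter spawns two anim trails, each initialized from its own skeletal mesh bone or socket pair and updated from that mesh.
# Spawn-group masks give the trails dark navy or warm cream color and different opacity.
# Color is multiplied by time-of-day color, and alpha follows a fade curve scaled by the system's swing fade.
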NiagaraSystem /BRCosmetics/Weapons/FORT_Melee/Pickaxe_VelvetDeskCam/FX/NS_Pickaxe_VelvetDeskCam_Trail
dependencies {
 usedDataInterfaces [&NiagaraDataInterfaceCurve, &NiagaraDataInterfaceSkeletalMesh, &NiagaraDataInterfaceColorCurve]
 usedDynamicInputs [&ColorFromCurve, &FNTimeOfDayScalar]
 usedModules [
  &self:SystemSpawnScript.SystemScriptSource.SystemScriptGraph.NiagaraNodeAssignment_0.SetVariables_#1
  &SystemState
  &self:SystemSpawnScript.SystemScriptSource.SystemScriptGraph.NiagaraNodeAssignment_1.SetVariables_#2
  &self:SystemSpawnScript.SystemScriptSource.SystemScriptGraph.NiagaraNodeAssignment_2.SetVariables_#3
 ]
}
emitters {Trail {
  emitterSpawnScript [&FN_RibbonIDOffsetUtility()]
  emitterUpdateScript [
   &EmitterState()
   &FN_SpawnAnimTrail(AnimTrailWidthMode=::FromSecond, Number of Anim Trails=2)
  ]
  particleSpawnScript [
   &FN_InitializeAnimTrailBoneIndecies()
   &FN_InitializeAnimTrailBoneIndecies(AnimTrail Selection=1, First Bone or Socket Index=3, Second Bone or Socket Index=2) {module=FN_InitializeAnimTrailBoneIndecies001}
   &FN_InitializeAnimTrail(
    AnimTrailWidth=1.4
    Color=&MultiplyLinearColor(
     Linear Color=<rgba 0.0125 0.0176 0.0802 1>
     Scale Factor=&MakeVector4_FromLinearColor(LinearColor=$System.TODColor)
    )
    Lifetime=0.11
    SkeletalMesh=$System.Skeletal Mesh
   )
  ]
  particleUpdateScript [
   &FN_UpdateAnimTrail(SkeletalMesh=$System.Skeletal Mesh)
   /SetVariables(Particles.HardColor:LinearColor=&MakeLinearColorFromVectorAndFloat(
     "Vector (RGB)"=&Multiply_Vector(
      A=&MakeVectorFromLinearColorRGB(Color=&MaskLinearColorBySpawnGroup(Masked Linear Color=<rgba 0.0125 0.0176 0.0802 1>, Passthrough Linear Color=<rgba 0.7011 0.6115 0.4784 1>))
      B=&MakeVectorFromLinearColorRGB(Color=$System.TODColor)
     )
     "Float (Alpha)"=&Multiply_Float(
      A=&FloatFromCurve(
       FloatCurve=DataInterface() {propertyValues {Curve {keys [{interpMode=RCIM_Cubic, leaveTangent=0.0199, tangentMode=RCTM_Break, time=0, value=1}, <at 0.6 0>]}}}
       Scale Curve=$System.SwingGeneralFade
      )
      B=&MaskFloatBySpawnGroup(Masked Float=1.3, Passthrough Float=0.3)
     )
    )) {isSetParameters=true}
  ]
  rendererClasses [&NiagaraRibbonRendererProperties]
  renderers [{rendererClass=&NiagaraRibbonRendererProperties}]
 }}
summary {
 systemName=NS_Pickaxe_VelvetDeskCam_Trail
 emitters [{emitterName=Trail, rendererClasses [&NiagaraRibbonRendererProperties]}]
 userVariables [User.RibbonIDOffset {type=NiagaraInt32}]
}
