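# Physics asset for the VelvetDeskCam pickaxe, giving its four ball attachments bodies and joints for physical motion.
# Each ball is a two-link chain: a small Base capsule constrained to the b_AR_Root capsule, then a Dyn sphere constrained to that Base.
# Base-to-root joints set swing and twist free, while Dyn-to-Base joints carry much wider angular limit values.
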PhysicsAsset /BRCosmetics/Weapons/FORT_Melee/Pickaxe_VelvetDeskCam/Meshes/VelvetDeskCam_Axe_PhysicsAsset
bodies {
 Ball_01_Base [{shapeType=Capsule, length=2, radius=2, shapeName=Ball_01_Base_capsule, rotation=<rot 90 0 0>}]
 Ball_01_Dyn [{shapeType=Sphere, radius=4.6472, shapeName=Ball_01_Dyn_sphere}]
 Ball_02_Base [{shapeType=Capsule, length=2, radius=2, shapeName=Ball_02_Base_capsule, rotation=<rot 90 0 0>}]
 Ball_02_Dyn [{shapeType=Sphere, radius=3.7029, shapeName=Ball_02_Dyn_sphere}]
 Ball_03_Base [{shapeType=Capsule, length=2, radius=2, shapeName=Ball_03_Base_capsule, rotation=<rot 90 0 0>}]
 Ball_03_Dyn [{shapeType=Sphere, radius=3.8034, shapeName=Ball_03_Dyn_sphere}]
 Ball_04_Base [{shapeType=Capsule, length=2, radius=2, shapeName=Ball_04_Base_capsule, rotation=<rot 90 0 0>}]
 Ball_04_Dyn [{shapeType=Sphere, radius=4.6472, shapeName=Ball_01_Dyn_sphere}]
 b_AR_Root [{shapeType=Capsule, length=1.0742, radius=3.3215, center=<6.5851 0 24.449>, rotation=<rot 90 0 0>}]
}
constraints [
 {
  bone1Name=Ball_01_Base
  bone2Name=b_AR_Root
  swing1LimitDegrees=45
  swing1Motion=Free
  swing2LimitDegrees=45
  swing2Motion=Free
  twistLimitDegrees=45
  twistMotion=Free
 }
 {bone1Name=Ball_01_Dyn, bone2Name=Ball_01_Base, swing1LimitDegrees=180, swing2LimitDegrees=180, twistLimitDegrees=180}
 {
  bone1Name=Ball_02_Base
  bone2Name=b_AR_Root
  swing1LimitDegrees=45
  swing1Motion=Free
  swing2LimitDegrees=45
  swing2Motion=Free
  twistLimitDegrees=45
  twistMotion=Free
 }
 {bone1Name=Ball_02_Dyn, bone2Name=Ball_02_Base, swing1LimitDegrees=180, swing2LimitDegrees=180, twistLimitDegrees=180}
 {
  bone1Name=Ball_03_Base
  bone2Name=b_AR_Root
  swing1LimitDegrees=45
  swing1Motion=Free
  swing2LimitDegrees=45
  swing2Motion=Free
  twistLimitDegrees=45
  twistMotion=Free
 }
 {
  bone1Name=Ball_04_Base
  bone2Name=b_AR_Root
  swing1LimitDegrees=45
  swing1Motion=Free
  swing2LimitDegrees=45
  swing2Motion=Free
  twistLimitDegrees=45
  twistMotion=Free
 }
 {bone1Name=Ball_04_Dyn, bone2Name=Ball_04_Base, swing1LimitDegrees=180, swing2LimitDegrees=180, twistLimitDegrees=180}
 {bone1Name=Ball_03_Dyn, bone2Name=Ball_03_Base, swing1LimitDegrees=180, swing2LimitDegrees=180, twistLimitDegrees=180}
]
